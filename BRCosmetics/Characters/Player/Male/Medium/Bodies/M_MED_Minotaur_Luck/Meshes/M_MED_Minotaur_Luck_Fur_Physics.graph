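# Physics asset for the fur on the Minotaur Luck male medium player body, defining capsule bodies for secondary motion of back and shoulder fur bones.
# Back fur bodies constrain directly to a spine_05 anchor capsule with limited swing and very tight twist.
# Each shoulder tuft hangs off its own small parent body, which rotates freely relative to spine_05, while the tuft itself has narrow swing and twist limits against that parent.
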PhysicsAsset /BRCosmetics/Characters/Player/Male/Medium/Bodies/M_MED_Minotaur_Luck/Meshes/M_MED_Minotaur_Luck_Fur_Physics
bodies {
 dyn_fur_back_01 [{shapeType=Capsule, length=2.25, radius=2.25, center=<1 0 0>, rotation=<rot 90 0 0>}]
 dyn_fur_back_010 [{shapeType=Capsule, length=2.75, radius=1.75, center=<1 0 0>, rotation=<rot 90 0 0>}]
 dyn_fur_back_011 [{shapeType=Capsule, length=2.25, radius=1.75, center=<1 0 0>, rotation=<rot 90 0 0>}]
 dyn_fur_back_02 [{shapeType=Capsule, length=2.25, radius=1.5, center=<1 0 0>, rotation=<rot 90 0 0>}]
 dyn_fur_back_03 [{shapeType=Capsule, length=3.25, radius=3, center=<1.2 0 0>, rotation=<rot 90 0 0>}]
 dyn_fur_back_04 [{shapeType=Capsule, length=2.25, radius=2.25, center=<1 0 0>, rotation=<rot 90 0 0>}]
 dyn_fur_back_05 [{shapeType=Capsule, length=2.75, radius=2.25, center=<1 0 0>, rotation=<rot 90 0 0>}]
 dyn_fur_back_06 [{shapeType=Capsule, length=2.75, radius=1.75, center=<1 0 0>, rotation=<rot 90 0 0>}]
 dyn_fur_back_07 [{shapeType=Capsule, length=1.25, radius=1.5, center=<1 0 0>, rotation=<rot 90 0 0>}]
 dyn_fur_back_08 [{shapeType=Capsule, length=2, radius=2.75, center=<2.7935 0 0>, rotation=<rot 90 0 0>}]
 dyn_fur_back_09 [{shapeType=Capsule, length=2.25, radius=3.5, center=<1 0 0>, rotation=<rot 90 0 0>}]
 dyn_fur_shldr_01_l [{shapeType=Capsule, length=1.75, radius=2.5, center=<0.5 0 0>, rotation=<rot 90 0 0>}]
 dyn_fur_shldr_01_r [{shapeType=Capsule, length=1, radius=1, center=<0.8 0 0>, rotation=<rot 90 0 0>}]
 dyn_fur_shldr_02_l [{shapeType=Capsule, length=0.75, radius=2, center=<0.5 0 0>, rotation=<rot 90 0 0>}]
 dyn_fur_shldr_02_r [{shapeType=Capsule, length=1.75, radius=2.75, center=<0.5 0 0>, rotation=<rot 90 0 0>}]
 dyn_fur_shldr_03_l [{shapeType=Capsule, length=1.75, radius=2.5, center=<0.5 0 0>, rotation=<rot 90 0 0>}]
 dyn_fur_shldr_03_r [{shapeType=Capsule, length=4, radius=1.75, center=<0.5 0 0>, rotation=<rot 90 0 0>}]
 dyn_fur_shldr_04_l [{shapeType=Capsule, length=1.75, radius=2, center=<0.8 0 0>, rotation=<rot 90 0 0>}]
 dyn_fur_shldr_04_r [{shapeType=Capsule, length=3.25, radius=2.5, center=<1 0 0>, rotation=<rot 90 0 0>}]
 dyn_fur_shldr_05_r [{shapeType=Capsule, length=4, radius=1.75, center=<0.3 0 0>, rotation=<rot 90 0 0>}]
 dyn_fur_shldr_06_l [{shapeType=Capsule, length=1.25, radius=2, center=<0.3 0 0>, rotation=<rot 90 0 0>}]
 dyn_parent_fur_shldr_01_l [{shapeType=Capsule, length=0.505, radius=0.505, rotation=<rot 0 -90 0>}]
 dyn_parent_fur_shldr_01_r [{shapeType=Capsule, length=0.505, radius=0.505, rotation=<rot 0 -90 0>}]
 dyn_parent_fur_shldr_02_l [{shapeType=Capsule, length=0.505, radius=0.505, rotation=<rot 0 -90 0>}]
 dyn_parent_fur_shldr_02_r [{shapeType=Capsule, length=0.505, radius=0.505, rotation=<rot 0 -90 0>}]
 dyn_parent_fur_shldr_03_l [{shapeType=Capsule, length=0.505, radius=0.505, rotation=<rot 0 -90 0>}]
 dyn_parent_fur_shldr_03_r [{shapeType=Capsule, length=0.505, radius=0.505, rotation=<rot 0 -90 0>}]
 dyn_parent_fur_shldr_04_l [{shapeType=Capsule, length=0.505, radius=0.505, rotation=<rot 0 -90 0>}]
 dyn_parent_fur_shldr_04_r [{shapeType=Capsule, length=0.505, radius=0.505, rotation=<rot 0 -90 0>}]
 dyn_parent_fur_shldr_05_r [{shapeType=Capsule, length=0.505, radius=0.505, rotation=<rot 0 -90 0>}]
 dyn_parent_fur_shldr_06_l [{shapeType=Capsule, length=0.505, radius=0.505, rotation=<rot 0 -90 0>}]
 spine_05 [{shapeType=Capsule, length=0, radius=2.8804, center=<7.706 -0.5587 1.0057>, rotation=<rot 0 -1.9303 -81.0399>}]
}
constraints [
 {bone1Name=dyn_fur_back_01, bone2Name=spine_05, swing1LimitDegrees=20, swing2LimitDegrees=20, twistLimitDegrees=3}
 {bone1Name=dyn_fur_back_02, bone2Name=spine_05, swing1LimitDegrees=20, swing2LimitDegrees=20, twistLimitDegrees=3}
 {bone1Name=dyn_fur_back_03, bone2Name=spine_05, swing1LimitDegrees=20, swing2LimitDegrees=20, twistLimitDegrees=3}
 {bone1Name=dyn_fur_back_04, bone2Name=spine_05, swing1LimitDegrees=20, swing2LimitDegrees=20, twistLimitDegrees=3}
 {bone1Name=dyn_fur_back_05, bone2Name=spine_05, swing1LimitDegrees=20, swing2LimitDegrees=20, twistLimitDegrees=3}
 {bone1Name=dyn_fur_back_06, bone2Name=spine_05, swing1LimitDegrees=20, swing2LimitDegrees=20, twistLimitDegrees=3}
 {bone1Name=dyn_fur_back_07, bone2Name=spine_05, swing1LimitDegrees=20, swing2LimitDegrees=20, twistLimitDegrees=3}
 {bone1Name=dyn_fur_back_08, bone2Name=spine_05, swing1LimitDegrees=45.2857, swing2LimitDegrees=35.1429, twistLimitDegrees=3}
 {bone1Name=dyn_fur_back_09, bone2Name=spine_05, swing1LimitDegrees=25.1429, swing2LimitDegrees=33.7143, twistLimitDegrees=3}
 {bone1Name=dyn_fur_back_010, bone2Name=spine_05, swing1LimitDegrees=28.5714, swing2LimitDegrees=20, twistLimitDegrees=3}
 {bone1Name=dyn_fur_back_011, bone2Name=spine_05, swing1LimitDegrees=20, swing2LimitDegrees=20, twistLimitDegrees=3}
 {
  bone1Name=dyn_parent_fur_shldr_01_r
  bone2Name=spine_05
  swing1LimitDegrees=45
  swing1Motion=Free
  swing2LimitDegrees=45
  swing2Motion=Free
  twistLimitDegrees=45
  twistMotion=Free
 }
 {
  bone1Name=dyn_parent_fur_shldr_02_r
  bone2Name=spine_05
  swing1LimitDegrees=45
  swing1Motion=Free
  swing2LimitDegrees=45
  swing2Motion=Free
  twistLimitDegrees=45
  twistMotion=Free
 }
 {
  bone1Name=dyn_parent_fur_shldr_03_r
  bone2Name=spine_05
  swing1LimitDegrees=45
  swing1Motion=Free
  swing2LimitDegrees=45
  swing2Motion=Free
  twistLimitDegrees=45
  twistMotion=Free
 }
 {
  bone1Name=dyn_parent_fur_shldr_04_r
  bone2Name=spine_05
  swing1LimitDegrees=45
  swing1Motion=Free
  swing2LimitDegrees=45
  swing2Motion=Free
  twistLimitDegrees=45
  twistMotion=Free
 }
 {
  bone1Name=dyn_parent_fur_shldr_05_r
  bone2Name=spine_05
  swing1LimitDegrees=45
  swing1Motion=Free
  swing2LimitDegrees=45
  swing2Motion=Free
  twistLimitDegrees=45
  twistMotion=Free
 }
 {bone1Name=dyn_fur_shldr_01_r, bone2Name=dyn_parent_fur_shldr_01_r, swing1LimitDegrees=15, swing2LimitDegrees=15, twistLimitDegrees=5}
 {bone1Name=dyn_fur_shldr_02_r, bone2Name=dyn_parent_fur_shldr_02_r, swing1LimitDegrees=15, swing2LimitDegrees=15, twistLimitDegrees=5}
 {bone1Name=dyn_fur_shldr_03_r, bone2Name=dyn_parent_fur_shldr_03_r, swing1LimitDegrees=15, swing2LimitDegrees=15, twistLimitDegrees=5}
 {
  bone1Name=dyn_fur_shldr_04_r
  bone2Name=dyn_parent_fur_shldr_04_r
  swing1LimitDegrees=33.8571
  swing2LimitDegrees=33.8571
  twistLimitDegrees=5
 }
 {bone1Name=dyn_fur_shldr_05_r, bone2Name=dyn_parent_fur_shldr_05_r, swing1LimitDegrees=15, swing2LimitDegrees=15, twistLimitDegrees=5}
 {
  bone1Name=dyn_parent_fur_shldr_01_l
  bone2Name=spine_05
  swing1LimitDegrees=45
  swing1Motion=Free
  swing2LimitDegrees=45
  swing2Motion=Free
  twistLimitDegrees=45
  twistMotion=Free
 }
 {
  bone1Name=dyn_parent_fur_shldr_02_l
  bone2Name=spine_05
  swing1LimitDegrees=45
  swing1Motion=Free
  swing2LimitDegrees=45
  swing2Motion=Free
  twistLimitDegrees=45
  twistMotion=Free
 }
 {
  bone1Name=dyn_parent_fur_shldr_03_l
  bone2Name=spine_05
  swing1LimitDegrees=45
  swing1Motion=Free
  swing2LimitDegrees=45
  swing2Motion=Free
  twistLimitDegrees=45
  twistMotion=Free
 }
 {
  bone1Name=dyn_parent_fur_shldr_04_l
  bone2Name=spine_05
  swing1LimitDegrees=45
  swing1Motion=Free
  swing2LimitDegrees=45
  swing2Motion=Free
  twistLimitDegrees=45
  twistMotion=Free
 }
 {
  bone1Name=dyn_parent_fur_shldr_06_l
  bone2Name=spine_05
  swing1LimitDegrees=45
  swing1Motion=Free
  swing2LimitDegrees=45
  swing2Motion=Free
  twistLimitDegrees=45
  twistMotion=Free
 }
 {bone1Name=dyn_fur_shldr_01_l, bone2Name=dyn_parent_fur_shldr_01_l, swing1LimitDegrees=20, swing2LimitDegrees=20, twistLimitDegrees=5}
 {bone1Name=dyn_fur_shldr_02_l, bone2Name=dyn_parent_fur_shldr_02_l, swing1LimitDegrees=15, swing2LimitDegrees=15, twistLimitDegrees=5}
 {bone1Name=dyn_fur_shldr_03_l, bone2Name=dyn_parent_fur_shldr_03_l, swing1LimitDegrees=20, swing2LimitDegrees=20, twistLimitDegrees=5}
 {bone1Name=dyn_fur_shldr_04_l, bone2Name=dyn_parent_fur_shldr_04_l, swing1LimitDegrees=20, swing2LimitDegrees=20, twistLimitDegrees=5}
 {bone1Name=dyn_fur_shldr_06_l, bone2Name=dyn_parent_fur_shldr_06_l, swing1LimitDegrees=20, swing2LimitDegrees=20, twistLimitDegrees=5}
]
