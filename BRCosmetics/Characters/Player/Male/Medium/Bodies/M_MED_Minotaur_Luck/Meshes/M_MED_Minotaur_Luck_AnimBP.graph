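# Post-process style Animation Blueprint for the Minotaur Luck cosmetic body, adding secondary motion to shoulder fur, shirt and armband.
# It copies the attached parent mesh's pose, then weighted constraint nodes drive fur and shirt helper bones from nearby spine, neck, arm, thigh and fur bones.
# Spring-based AnimDynamics then move three armband bones.
# Two RigidBody passes follow, using the body and fur physics assets; no gameplay logic is present.
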AnimBlueprint /BRCosmetics/Characters/Player/Male/Medium/Bodies/M_MED_Minotaur_Luck/Meshes/M_MED_Minotaur_Luck_AnimBP
target_skeleton=&/M_MED_Minotaur_Luck_Skeleton
graphs [
 AnimGraph {nodes [
   n1 = "Constraint - dyn_parent_fur_shldr_01_l (PPPPPP)"!(
    ComponentPose=Local To Component(LocalPose=Copy Pose From Mesh() {bUseAttachedParent=true})
    Alpha=1
    ConstraintWeights_0=0.2401
    ConstraintWeights_1:float=0.0313
    ConstraintWeights_2:float=0.3901
    ConstraintWeights_3:float=0.0419
    ConstraintWeights_4:float=0.0375
    ConstraintWeights_5:float=0.259
   ) -> (Pose) {
    alphaBoolBlend {blendOption=Linear}
    boneToModify {boneName=dyn_parent_fur_shldr_01_l}
    constraintSetup [
     {targetBone {boneName=dyn_fur_shldr_01_l}}
     {targetBone {boneName=dyn_fur_shldr_06_l}}
     {targetBone {boneName=clavicle_l}}
     {targetBone {boneName=dyn_fur_shldr_02_l}}
     {targetBone {boneName=spine_03}}
     {targetBone {boneName=spine_05}}
    ]
    constraintWeights [1,1,1,1,1,1]
   }
   n2 = "Constraint - dyn_parent_fur_shldr_02_l (PPPPPPP)"!(
    ComponentPose=@n1
    Alpha=1
    ConstraintWeights_0=0.0115
    ConstraintWeights_1:float=0.2853
    ConstraintWeights_2:float=0.4431
    ConstraintWeights_3:float=0.0161
    ConstraintWeights_4:float=0.0371
    ConstraintWeights_5:float=0.0118
    ConstraintWeights_6:float=0.1951
   ) -> (Pose) {
    alphaBoolBlend {blendOption=Linear}
    boneToModify {boneName=dyn_parent_fur_shldr_02_l}
    constraintSetup [
     {targetBone {boneName=dyn_fur_shldr_04_l}}
     {targetBone {boneName=clavicle_l}}
     {targetBone {boneName=dyn_fur_shldr_02_l}}
     {targetBone {boneName=dyn_fur_shldr_03_l}}
     {targetBone {boneName=spine_03}}
     {targetBone {boneName=dyn_fur_back_07}}
     {targetBone {boneName=spine_05}}
    ]
    constraintWeights [1,1,1,1,1,1,1]
   }
   n3 = "Constraint - dyn_parent_fur_shldr_03_l (PPPPPP)"!(
    ComponentPose=@n2
    Alpha=1
    ConstraintWeights_0=0.4848
    ConstraintWeights_1:float=0.3336
    ConstraintWeights_2:float=0.046
    ConstraintWeights_3:float=0.0408
    ConstraintWeights_4:float=0.0251
    ConstraintWeights_5:float=0.0696
   ) -> (Pose) {
    alphaBoolBlend {blendOption=Linear}
    boneToModify {boneName=dyn_parent_fur_shldr_03_l}
    constraintSetup [
     {targetBone {boneName=clavicle_l}}
     {targetBone {boneName=dyn_fur_shldr_03_l}}
     {targetBone {boneName=spine_03}}
     {targetBone {boneName=dyn_fur_shldr_04_l}}
     {targetBone {boneName=upperarm_twist_01_l}}
     {targetBone {boneName=spine_05}}
    ]
    constraintWeights [1,1,1,1,1,1]
   }
   n4 = "Constraint - dyn_parent_fur_shldr_04_l (PPPPPP)"!(
    ComponentPose=@n3
    Alpha=1
    ConstraintWeights_0=0.5056
    ConstraintWeights_1:float=0.0914
    ConstraintWeights_2:float=0.0477
    ConstraintWeights_3:float=0.2065
    ConstraintWeights_4:float=0.0128
    ConstraintWeights_5:float=0.1359
   ) -> (Pose) {
    alphaBoolBlend {blendOption=Linear}
    boneToModify {boneName=dyn_parent_fur_shldr_04_l}
    constraintSetup [
     {targetBone {boneName=clavicle_l}}
     {targetBone {boneName=spine_03}}
     {targetBone {boneName=spine_04}}
     {targetBone {boneName=dyn_fur_shldr_04_l}}
     {targetBone {boneName=upperarm_twist_01_l}}
     {targetBone {boneName=spine_05}}
    ]
    constraintWeights [1,1,1,1,1,1]
   }
   n5 = "Constraint - dyn_parent_fur_shldr_05_l (PPPPPP)"!(
    ComponentPose=@n4
    Alpha=1
    ConstraintWeights_0=0.3532
    ConstraintWeights_1:float=0.0989
    ConstraintWeights_2:float=0.0778
    ConstraintWeights_3:float=0.2896
    ConstraintWeights_4:float=0.0666
    ConstraintWeights_5:float=0.1138
   ) -> (Pose) {
    alphaBoolBlend {blendOption=Linear}
    boneToModify {boneName=dyn_parent_fur_shldr_05_l}
    constraintSetup [
     {targetBone {boneName=clavicle_l}}
     {targetBone {boneName=spine_03}}
     {targetBone {boneName=dyn_fur_back_011}}
     {targetBone {boneName=spine_05}}
     {targetBone {boneName=spine_04}}
     {targetBone {boneName=dyn_fur_shldr_04_l}}
    ]
    constraintWeights [1,1,1,1,1,1]
   }
   n6 = "Constraint - dyn_parent_fur_shldr_06_l (PPPPPPP)"!(
    ComponentPose=@n5
    Alpha=1
    ConstraintWeights_0=0.0308
    ConstraintWeights_1:float=0.2171
    ConstraintWeights_2:float=0.4499
    ConstraintWeights_3:float=0.0401
    ConstraintWeights_4:float=0.0205
    ConstraintWeights_5:float=0.2233
    ConstraintWeights_6:float=0.0183
   ) -> (Pose) {
    alphaBoolBlend {blendOption=Linear}
    boneToModify {boneName=dyn_parent_fur_shldr_06_l}
    constraintSetup [
     {targetBone {boneName=dyn_fur_shldr_01_l}}
     {targetBone {boneName=dyn_fur_shldr_06_l}}
     {targetBone {boneName=clavicle_l}}
     {targetBone {boneName=spine_03}}
     {targetBone {boneName=spine_04}}
     {targetBone {boneName=spine_05}}
     {targetBone {boneName=upperarm_twist_01_l}}
    ]
    constraintWeights [1,1,1,1,1,1,1]
   }
   n7 = "Constraint - dyn_parent_fur_shldr_01_r (PPPPPPP)"!(
    ComponentPose=@n6
    Alpha=1
    ConstraintWeights_0=0.1086
    ConstraintWeights_1:float=0.0365
    ConstraintWeights_2:float=0.0104
    ConstraintWeights_3:float=0.043
    ConstraintWeights_4:float=0.1527
    ConstraintWeights_5:float=0.4887
    ConstraintWeights_6:float=0.1601
   ) -> (Pose) {
    alphaBoolBlend {blendOption=Linear}
    boneToModify {boneName=dyn_parent_fur_shldr_01_r}
    constraintSetup [
     {targetBone {boneName=dyn_fur_shldr_05_r}}
     {targetBone {boneName=upperarm_twist_01_r}}
     {targetBone {boneName=dyn_fur_shldr_03_r}}
     {targetBone {boneName=spine_03}}
     {targetBone {boneName=spine_05}}
     {targetBone {boneName=clavicle_r}}
     {targetBone {boneName=dyn_fur_shldr_01_r}}
    ]
    constraintWeights [1,1,1,1,1,1,1]
   }
   n8 = "Constraint - dyn_parent_fur_shldr_02_r (PPPPPPPP)"!(
    ComponentPose=@n7
    Alpha=1
    ConstraintWeights_0=0.0151
    ConstraintWeights_1:float=0.1774
    ConstraintWeights_2:float=0.0315
    ConstraintWeights_3:float=0.1771
    ConstraintWeights_4:float=0.3294
    ConstraintWeights_5:float=0.2254
    ConstraintWeights_6:float=0.0132
    ConstraintWeights_7:float=0.0309
   ) -> (Pose) {
    alphaBoolBlend {blendOption=Linear}
    boneToModify {boneName=dyn_parent_fur_shldr_02_r}
    constraintSetup [
     {targetBone {boneName=dyn_fur_shldr_04_r}}
     {targetBone {boneName=dyn_fur_shldr_03_r}}
     {targetBone {boneName=spine_03}}
     {targetBone {boneName=spine_05}}
     {targetBone {boneName=dyn_fur_shldr_02_r}}
     {targetBone {boneName=clavicle_r}}
     {targetBone {boneName=neck_02}}
     {targetBone {boneName=neck_01}}
    ]
    constraintWeights [1,1,1,1,1,1,1,1]
   }
   n9 = "Constraint - dyn_parent_fur_shldr_03_r (PPPPPPP)"!(
    ComponentPose=@n8
    Alpha=1
    ConstraintWeights_0=0.0685
    ConstraintWeights_1:float=0.3235
    ConstraintWeights_2:float=0.0429
    ConstraintWeights_3:float=0.1821
    ConstraintWeights_4:float=0.2677
    ConstraintWeights_5:float=0.1023
    ConstraintWeights_6:float=0.013
   ) -> (Pose) {
    alphaBoolBlend {blendOption=Linear}
    boneToModify {boneName=dyn_parent_fur_shldr_03_r}
    constraintSetup [
     {targetBone {boneName=dyn_fur_shldr_04_r}}
     {targetBone {boneName=dyn_fur_shldr_03_r}}
     {targetBone {boneName=spine_03}}
     {targetBone {boneName=spine_05}}
     {targetBone {boneName=clavicle_r}}
     {targetBone {boneName=dyn_fur_shldr_02_r}}
     {targetBone {boneName=neck_01}}
    ]
    constraintWeights [1,1,1,1,1,1,1]
   }
   n10 = "Constraint - dyn_parent_fur_shldr_04_r (PPPPPPP)"!(
    ComponentPose=@n9
    Alpha=1
    ConstraintWeights_0=0.1934
    ConstraintWeights_1:float=0.0335
    ConstraintWeights_2:float=0.0825
    ConstraintWeights_3:float=0.0383
    ConstraintWeights_4:float=0.0568
    ConstraintWeights_5:float=0.0948
    ConstraintWeights_6:float=0.5007
   ) -> (Pose) {
    alphaBoolBlend {blendOption=Linear}
    boneToModify {boneName=dyn_parent_fur_shldr_04_r}
    constraintSetup [
     {targetBone {boneName=dyn_fur_shldr_05_r}}
     {targetBone {boneName=upperarm_twist_01_r}}
     {targetBone {boneName=dyn_fur_shldr_04_r}}
     {targetBone {boneName=dyn_fur_shldr_03_r}}
     {targetBone {boneName=spine_03}}
     {targetBone {boneName=spine_05}}
     {targetBone {boneName=clavicle_r}}
    ]
    constraintWeights [1,1,1,1,1,1,1]
   }
   n11 = "Constraint - dyn_parent_fur_shldr_05_r (PPPPP)"!(
    ComponentPose=@n10
    Alpha=1
    ConstraintWeights_0=0.4305
    ConstraintWeights_1:float=0.0342
    ConstraintWeights_2:float=0.4284
    ConstraintWeights_3:float=0.0423
    ConstraintWeights_4:float=0.0646
   ) -> (Pose) {
    alphaBoolBlend {blendOption=Linear}
    boneToModify {boneName=dyn_parent_fur_shldr_05_r}
    constraintSetup [
     {targetBone {boneName=clavicle_r}}
     {targetBone {boneName=upperarm_twist_01_r}}
     {targetBone {boneName=dyn_fur_shldr_05_r}}
     {targetBone {boneName=spine_03}}
     {targetBone {boneName=spine_05}}
    ]
    constraintWeights [1,1,1,1,1]
   }
   n12 = "Constraint - dyn_thigh_shirt_l (P)"(ComponentPose=@n11, Alpha=1, ConstraintWeights_0=1) {
    alphaBoolBlend {blendOption=Linear}
    boneToModify {boneName=dyn_thigh_shirt_l}
    constraintSetup [{targetBone {boneName=thigh_twist_01_l}}]
    constraintWeights [1]
   }
   n13 = "Constraint - dyn_thigh_shirt_r (P)"(ComponentPose=@n12, Alpha=1, ConstraintWeights_0=1) {
    alphaBoolBlend {blendOption=Linear}
    boneToModify {boneName=dyn_thigh_shirt_r}
    constraintSetup [{targetBone {boneName=thigh_twist_01_r}}]
    constraintWeights [1]
   }
   n14 = "Constraint - dyn_parent_shirt_frt_l (PPPP)"!(
    ComponentPose=@n13
    Alpha=0.3
    ConstraintWeights_0=0.0266
    ConstraintWeights_1:float=0.4551
    ConstraintWeights_2:float=0.0419
    ConstraintWeights_3:float=0.4765
   ) -> (Pose) {
    alphaBoolBlend {blendOption=Linear}
    boneToModify {boneName=dyn_parent_shirt_frt_l}
    constraintSetup [
     {targetBone {boneName=spine_01}}
     {targetBone {boneName=thigh_twist_01_l}}
     {targetBone {boneName=thigh_twist_01_r}}
     {targetBone {boneName=pelvis}}
    ]
    constraintWeights [1,1,1,1]
   }
   n15 = Anim Dynamics - Bone: dyn_armband_knot(ComponentPose=@n14, Alpha=0.75, ExternalForce="0,0,0") {
    alpha=0.9
    angularDampingOverride=1
    angularSpringConstant=1500
    bAngularSpring=true
    bGravityOverrideInSimSpace=true
    bOverrideAngularDamping=true
    numSolverIterationsPostUpdate=4
    numSolverIterationsPreUpdate=16
    simulationSpace=Actor
    alphaBoolBlend {blendOption=Linear}
    boundBone {boneName=dyn_armband_knot}
    physicsBodyDefinitions [{
      sphereCollisionRadius=0
      boundBone {boneName=dyn_armband_knot}
      boxExtents=<15 5 5>
      constraintSetup {angularTargetAxis=AxisZ, twistAxis=AxisY, angularLimitsMax=<15 0 15>, angularLimitsMin=<-15 0 -15>, angularTarget=<0 0 1>}
      localJointOffset=<0 0 0>
     }]
   }
   n16 = Anim Dynamics - Bone: dyn_armband_low(ComponentPose=@n15, Alpha=0.5, ExternalForce="0,0,0") {
    alpha=0.9
    angularDampingOverride=1
    angularSpringConstant=1500
    bAngularSpring=true
    bGravityOverrideInSimSpace=true
    bOverrideAngularDamping=true
    numSolverIterationsPostUpdate=4
    numSolverIterationsPreUpdate=16
    simulationSpace=Actor
    alphaBoolBlend {blendOption=Linear}
    boundBone {boneName=dyn_armband_low}
    physicsBodyDefinitions [{
      sphereCollisionRadius=0
      boundBone {boneName=dyn_armband_low}
      boxExtents=<15 5 5>
      constraintSetup {angularTargetAxis=AxisZ, twistAxis=AxisY, angularLimitsMax=<15 15 15>, angularLimitsMin=<-15 -15 -15>, angularTarget=<0 0 1>}
      localJointOffset=<0 0 0>
     }]
   }
   n17 = Anim Dynamics - Bone: dyn_armband_hi(ComponentPose=@n16, Alpha=0.5, ExternalForce="0,0,0") {
    alpha=0.9
    angularDampingOverride=1
    angularSpringConstant=1500
    bAngularSpring=true
    bGravityOverrideInSimSpace=true
    bOverrideAngularDamping=true
    numSolverIterationsPostUpdate=4
    numSolverIterationsPreUpdate=16
    simulationSpace=Actor
    alphaBoolBlend {blendOption=Linear}
    boundBone {boneName=dyn_armband_hi}
    physicsBodyDefinitions [{
      sphereCollisionRadius=0
      boundBone {boneName=dyn_armband_hi}
      boxExtents=<15 5 5>
      constraintSetup {angularTargetAxis=AxisZ, twistAxis=AxisY, angularLimitsMax=<15 15 15>, angularLimitsMin=<-15 -15 -15>, angularTarget=<0 0 1>}
      localJointOffset=<0 0 0>
     }]
   }
   n18 = RigidBody(ComponentPose=@n17) {
    bClampLinearTranslationLimitToRefPose=true
    bOverrideWorldGravity=true
    overridePhysicsAsset=&/M_MED_Minotaur_Luck_Physics
    simulationSpace=BaseBoneSpace
    alphaBoolBlend {blendOption=Linear}
    baseBoneRef {boneName=root}
    componentAppliedLinearAccClamp=<1100 1100 1100>
    componentLinearAccScale=<0.5 0.5 0.5>
    componentLinearVelScale=<0.5 0.5 0.5>
    overrideWorldGravity {z=-980}
   }
   n19 = RigidBody(ComponentPose=@n18) {
    bClampLinearTranslationLimitToRefPose=true
    bOverrideWorldGravity=true
    overridePhysicsAsset=&/M_MED_Minotaur_Luck_Fur_Physics
    simulationSpace=BaseBoneSpace
    alphaBoolBlend {blendOption=Linear}
    baseBoneRef {boneName=root}
    componentAppliedLinearAccClamp=<1100 1100 1100>
    componentLinearAccScale=<0.5 0.5 0.7>
    componentLinearVelScale=<0.5 0.5 0.7>
    overrideWorldGravity {z=-980}
   }
   Output Pose(Result=Component To Local(ComponentPose=@n19)) {layerGroup=None}
  ]}
 EventGraph {nodes [Event BlueprintUpdateAnimation()]}
]
variables []
