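# Physics asset for the face accessory mesh of the male medium Minotaur Luck player cosmetic, defining simulated motion for its dangling hair and tip bones.
# Small capsule bodies cover the head and three two-bone chains: back hair, left tip and right tip.
# Each chain is constrained to the head, with tighter swing limits at the root, looser swing at the second segment and little twist throughout.
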PhysicsAsset /BRCosmetics/Characters/Player/Male/Medium/Bodies/M_MED_Minotaur_Luck/Meshes/Parts/M_MED_Minotaur_Luck_FaceAcc_Physics
bodies {
 dyn_hair_bck_01 [{shapeType=Capsule, length=2.5, radius=2.75, center=<0 0 -1>}]
 dyn_hair_bck_02 [{shapeType=Capsule, length=1.5, radius=2, center=<0 0 -1>}]
 dyn_tip_01_l [{shapeType=Capsule, length=2.5, radius=2.75, center=<0 0 -1>}]
 dyn_tip_01_r [{shapeType=Capsule, length=2.5, radius=2.75, center=<0 0 -1>}]
 dyn_tip_02_l [{shapeType=Capsule, length=1.5, radius=2, center=<0 0 -1>}]
 dyn_tip_02_r [{shapeType=Capsule, length=1.5, radius=2, center=<0 0 -1>}]
 head [{shapeType=Capsule, length=1, radius=1}]
}
constraints [
 {bone1Name=dyn_hair_bck_01, bone2Name=head, swing1LimitDegrees=25, swing2LimitDegrees=25, twistLimitDegrees=5}
 {bone1Name=dyn_hair_bck_02, bone2Name=dyn_hair_bck_01, swing1LimitDegrees=45, swing2LimitDegrees=45, twistLimitDegrees=5}
 {bone1Name=dyn_tip_01_r, bone2Name=head, swing1LimitDegrees=25, swing2LimitDegrees=25, twistLimitDegrees=5}
 {bone1Name=dyn_tip_02_r, bone2Name=dyn_tip_01_r, swing1LimitDegrees=45, swing2LimitDegrees=45, twistLimitDegrees=5}
 {bone1Name=dyn_tip_01_l, bone2Name=head, swing1LimitDegrees=25, swing2LimitDegrees=25, twistLimitDegrees=5}
 {bone1Name=dyn_tip_02_l, bone2Name=dyn_tip_01_l, swing1LimitDegrees=45, swing2LimitDegrees=45, twistLimitDegrees=5}
]
